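; Season 15 Nightmare quest ability that tracks how long a player wearing the tagged outfit stays near hostile players, ending immediately without that outfit.
; A looping timer, randomized around one second, scans pawns within a radius and accumulates time when consecutive checks find an enemy.
; At three accumulated seconds it sends whole seconds as a quest stat event, keeping the remainder.
; Being downed cancels it, and ending runs a final partial-interval check.
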(event Ability|EventActivateAbility
  ;[1 1 1] Run first check and start looping check
  (|CallParentFunction)
  (if (GameplayTags|TagContainer|CompareTagAssetInterfacetoOtherTags (Variables|Default|GetOutfitTag) (Variables|GAVars|Data|GetOwningPlayerPawn))
    (Variables|Default|SetHas%20Outfit true)
    (CallFunction|CheckNearbyEnemies false)
    (bind _output_get (Variables|Default|SetCheck%20Interval (Math|Random|RandomFloatinRange 0.9 1.1)))
    (bind _returnvalue (Utilities|Time|SetTimerbyEvent (delegate Custom|LoopCheck) _output_get true))
    (Variables|Default|SetTimer%20Check%20Enemies _returnvalue)
    (EventDispatchers|BindEventtoOnPlayerStartDBNO (Variables|GAVars|Data|GetOwningPlayerPawn) (EventDispatchers|CreateEvent :!event K2_CancelAbility))
    (else
      (Variables|Default|SetHas%20Outfit false)
      (Ability|EndAbility))))

(event Ability|EventFailedToActivatePassiveAbility
  (params (ActorInfo :type Gameplay%20Ability%20Actor%20Info%20Structure)))

(event Custom|LoopCheck
  ;[1 1 1] Run first check and start looping check
  (CallFunction|CheckNearbyEnemies false))

(event Custom|Check%20Nearby%20Enemies
  (params (Last%20Check :type Boolean))
  (bind _accumulatedsuccesstime (Variables|Default|GetAccumulatedSuccessTime))
  (bind _returnvalue (Math|Float|Floor _accumulatedsuccesstime))
  (if (not Last%20Check)
    (Variables|Default|SetLastLoopTime (Fortnite|Sync|Time|GetServerWorldTimeSeconds)))
  (Variables|Default|SetCurrentCheckSucceeded false)
  (bind _outfortplayerpawns (Fortnite|GetFortPlayerPawnsInRange (Transformation|GetActorLocation (Variables|GAVars|Data|GetOwningPlayerPawn)) (* (Variables|Default|GetRadius%20in%20Meters) 1e+02)))
  ;[1 1 1] Check each hit actor until we find an enemy
  (bind _array_element (Utilities|Array|ForEachLoopwithBreak _outfortplayerpawns)
    (:LoopBody
      (if (Utilities|Enum|Equal(Enum) (Team|GetTeamAffiliation _array_element (Variables|GAVars|Data|GetOwningPlayerPawn)) "Hostile")
        (Variables|Default|SetCurrentCheckSucceeded true)))
    (:Completed
      _))
  (if (and (Variables|Default|GetLast%20Check%20Succeeded) (Variables|Default|GetCurrentCheckSucceeded))
    (Variables|Default|SetAccumulatedSuccessTime (+ (Variables|Default|GetAccumulatedSuccessTime) (select Last%20Check (- (Fortnite|Sync|Time|GetServerWorldTimeSeconds) (Variables|Default|GetLastLoopTime)) (Variables|Default|GetCheck%20Interval))) :!pinned true) ; Add the successful time
    (else
      _))
  (Variables|Default|SetLast%20Check%20Succeeded (Variables|Default|GetCurrentCheckSucceeded))
  ;[1 1 1] Update quest progress as we exceed int threshold, save off leftover successful time
  (if (>= (Variables|Default|GetAccumulatedSuccessTime) 3.0)
    (Quest|SendComplexCustomStatEvent (Variables|GAVars|Data|GetOwningPlayerQuestManager) 0 (Variables|Default|GetQuestTag) "(GameplayTags=)" _returnvalue)
    (Variables|Default|SetAccumulatedSuccessTime (- _accumulatedsuccesstime (Math|Conversions|ToFloat(Integer) _returnvalue)))))

(event Ability|EventOnEndAbility
  (params (bWasCancelled :type Boolean))
  ;[1 1 1] OnEndAbility: Kill timer and run one last check
  (|CallParentFunction bWasCancelled)
  (if (Variables|Default|GetHas%20Outfit)
    (Utilities|Time|ClearandInvalidateTimerbyHandle (Variables|Default|GetTimer%20Check%20Enemies))
    (CallFunction|CheckNearbyEnemies true)
    (EventDispatchers|UnbindEventfromOnPlayerStartDBNO (Variables|GAVars|Data|GetOwningPlayerPawn) (EventDispatchers|CreateEvent :!event K2_CancelAbility))))
