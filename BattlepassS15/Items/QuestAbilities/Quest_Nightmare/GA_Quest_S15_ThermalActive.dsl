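; Gameplay ability for a Season 15 Nightmare quest that flags the owning player with a "thermal active" gameplay effect.
; On event activation it checks the current weapon data's owned tags against the T_Weapon tags, ending immediately if none match.
; On a match it applies GE_Quest_S15_ThermalActive and stores the handle, which OnEndAbility uses to remove the effect.
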(event Ability|EventActivateAbilityFromEvent
  (params (EventData :type Gameplay%20Event%20Data%20Structure%20%28by%20ref%29))
  (|CallParentFunction EventData)
  (if (GameplayTags|HasAnyTags (GameplayTags|GetOwnedGameplayTags (Data|GetWeaponData (Variables|/Script/FortniteGame.FortPawn|GetCurrentWeapon (Variables|GAVars|Data|GetOwningPlayerPawn))) :!class /Script/GameplayTags.GameplayTagAssetInterface) (Variables|Default|GetT_Weapon) true)
    (bind _returnvalue (Ability|ApplyGameplayEffectToOwner "/BattlepassS15/Items/QuestAbilities/Quest_Nightmare/GE_Quest_S15_ThermalActive.GE_Quest_S15_ThermalActive_C"))
    (Variables|Default|SetGE_ThermalActive _returnvalue)
    (else
      (Ability|EndAbility))))

(event Ability|EventOnEndAbility
  (params (bWasCancelled :type Boolean))
  (Ability|RemoveGameplayEffectFromOwnerWithHandle (Variables|Default|GetGE_ThermalActive)))
